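# Physics asset for the M_MED_Desert_Shadow cosmetic backpack, providing bodies and joints for its dangling straps and bands.
# One large capsule covers the pack, with seven chains of small capsules jointed to it: a top strap, two straps and four bands.
# Both straps and bands 03–04 rotate freely at the pack, while other roots are limited.
# Links within chains have tight twist limits, with swing generally loosening toward strap tips.
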PhysicsAsset /BRCosmetics/Accessories/FORT_Backpacks/Backpack_M_MED_Desert_Shadow/Meshes/M_MED_Desert_Shadow_Pack_Physics
bodies {
 Top_strap_Base_dyn [{shapeType=Capsule, length=1, radius=1, shapeName=Top_strap_Base_dyn_capsule, rotation=<rot 90 0 0>}]
 Top_strap_dyn_01 [{shapeType=Capsule, length=1, radius=1, shapeName=Top_strap_dyn_01_capsule, center=<1 0 0>, rotation=<rot 90 0 0>}]
 Top_strap_dyn_02 [{shapeType=Capsule, length=1, radius=1, shapeName=Top_strap_dyn_02_capsule, center=<1 0 0>, rotation=<rot 90 0 0>}]
 Top_strap_dyn_03 [{shapeType=Capsule, length=1, radius=1, shapeName=Top_strap_dyn_03_capsule, center=<1 0 0>, rotation=<rot 90 0 0>}]
 Top_strap_dyn_04 [{shapeType=Capsule, length=1, radius=1, shapeName=Top_strap_dyn_04_capsule, center=<1 0 0>, rotation=<rot 90 0 0>}]
 Top_strap_dyn_05 [{shapeType=Capsule, length=1, radius=1, shapeName=Top_strap_dyn_05_capsule, center=<2.052 0 0>, rotation=<rot 90 0 0>}]
 band01_dyn_01 [{shapeType=Capsule, length=1, radius=1, shapeName=band01_dyn_01_capsule, center=<1 0 0>, rotation=<rot 90 0 0>}]
 band01_dyn_02 [{shapeType=Capsule, length=1, radius=1, shapeName=band01_dyn_02_capsule, center=<1 0 0>, rotation=<rot 90 0 0>}]
 band01_dyn_base [{shapeType=Capsule, length=1, radius=1, shapeName=band01_dyn_base_capsule, rotation=<rot 90 0 0>}]
 band02_dyn_01 [{shapeType=Capsule, length=1, radius=1, shapeName=band02_dyn_01_capsule, center=<1 0 0>, rotation=<rot 90 0 0>}]
 band02_dyn_02 [{shapeType=Capsule, length=1, radius=1, shapeName=band02_dyn_02_capsule, center=<1 0 0>, rotation=<rot 90 0 0>}]
 band02_dyn_base [{shapeType=Capsule, length=1, radius=1, shapeName=band02_dyn_base_capsule, rotation=<rot 90 0 0>}]
 band03_dyn_00 [{shapeType=Capsule, length=1, radius=1, shapeName=band03_dyn_00_capsule, center=<1 0 0>, rotation=<rot 90 0 0>}]
 band03_dyn_01 [{shapeType=Capsule, length=1, radius=1, shapeName=band03_dyn_01_capsule, center=<1 0 0>, rotation=<rot 90 0 0>}]
 band03_dyn_base [{shapeType=Capsule, length=1, radius=1, shapeName=band03_dyn_base_capsule, rotation=<rot 90 0 0>}]
 band04_dyn_00 [{shapeType=Capsule, length=1, radius=1, shapeName=band04_dyn_00_capsule, center=<1 0 0>, rotation=<rot 90 0 0>}]
 band04_dyn_01 [{shapeType=Capsule, length=1, radius=1, shapeName=band04_dyn_01_capsule, center=<1 0 0>, rotation=<rot 90 0 0>}]
 band04_dyn_base [{shapeType=Capsule, length=1, radius=1, shapeName=band04_dyn_base_capsule, rotation=<rot 90 0 0>}]
 pack [{
   shapeType=Capsule
   length=39.8604
   radius=8.9939
   shapeName=pack_capsule
   center=<-2.965 12.7014 -7.1927>
   rotation=<rot 28.58 -91.5442 24.7955>
  }]
 strap01_dyn_01 [{shapeType=Capsule, length=1, radius=1, shapeName=strap01_dyn_01_capsule, rotation=<rot 90 0 0>}]
 strap01_dyn_02 [{shapeType=Capsule, length=1, radius=1, shapeName=strap01_dyn_02_capsule, center=<0.5 0 0>, rotation=<rot 90 0 0>}]
 strap01_dyn_03 [{shapeType=Capsule, length=2.25, radius=1, shapeName=strap01_dyn_03_capsule, center=<1 0 0>, rotation=<rot 90 0 0>}]
 strap01_dyn_base [{shapeType=Capsule, length=1, radius=1, shapeName=strap01_dyn_base_capsule, rotation=<rot 90 0 0>}]
 strap02_dyn_01 [{shapeType=Capsule, length=1, radius=1, shapeName=strap02_dyn_01_capsule, rotation=<rot 90 0 0>}]
 strap02_dyn_02 [{shapeType=Capsule, length=1, radius=1, shapeName=strap02_dyn_02_capsule, center=<0.5 0 0>, rotation=<rot 90 0 0>}]
 strap02_dyn_03 [{shapeType=Capsule, length=2.5, radius=1, shapeName=strap02_dyn_03_capsule, center=<1 0 0>, rotation=<rot 90 0 0>}]
 strap02_dyn_base [{shapeType=Capsule, length=1, radius=1, shapeName=strap02_dyn_base_capsule, rotation=<rot 90 0 0>}]
}
constraints [
 {bone1Name=Top_strap_Base_dyn, bone2Name=pack, swing1LimitDegrees=45, swing2LimitDegrees=45, twistLimitDegrees=45}
 {bone1Name=Top_strap_dyn_01, bone2Name=Top_strap_Base_dyn, swing1LimitDegrees=5, swing2LimitDegrees=5, twistLimitDegrees=1}
 {bone1Name=Top_strap_dyn_02, bone2Name=Top_strap_dyn_01, swing1LimitDegrees=20, swing2LimitDegrees=15, twistLimitDegrees=1}
 {bone1Name=Top_strap_dyn_03, bone2Name=Top_strap_dyn_02, swing1LimitDegrees=25, swing2LimitDegrees=15, twistLimitDegrees=1}
 {bone1Name=Top_strap_dyn_04, bone2Name=Top_strap_dyn_03, swing1LimitDegrees=40, swing2LimitDegrees=25, twistLimitDegrees=1}
 {bone1Name=Top_strap_dyn_05, bone2Name=Top_strap_dyn_04, swing1LimitDegrees=45, swing2LimitDegrees=40, twistLimitDegrees=1}
 {
  bone1Name=strap01_dyn_base
  bone2Name=pack
  swing1LimitDegrees=45
  swing1Motion=Free
  swing2LimitDegrees=45
  swing2Motion=Free
  twistLimitDegrees=45
  twistMotion=Free
 }
 {bone1Name=strap01_dyn_01, bone2Name=strap01_dyn_base, swing1LimitDegrees=10, swing2LimitDegrees=10, twistLimitDegrees=1}
 {bone1Name=strap01_dyn_02, bone2Name=strap01_dyn_01, swing1LimitDegrees=30, swing2LimitDegrees=30, twistLimitDegrees=1}
 {bone1Name=strap01_dyn_03, bone2Name=strap01_dyn_02, swing1LimitDegrees=30, swing2LimitDegrees=30, twistLimitDegrees=1}
 {
  bone1Name=strap02_dyn_base
  bone2Name=pack
  swing1LimitDegrees=45
  swing1Motion=Free
  swing2LimitDegrees=45
  swing2Motion=Free
  twistLimitDegrees=45
  twistMotion=Free
 }
 {bone1Name=strap02_dyn_01, bone2Name=strap02_dyn_base, swing1LimitDegrees=10, swing2LimitDegrees=10, twistLimitDegrees=1}
 {bone1Name=strap02_dyn_02, bone2Name=strap02_dyn_01, swing1LimitDegrees=15, swing2LimitDegrees=15, twistLimitDegrees=1}
 {bone1Name=strap02_dyn_03, bone2Name=strap02_dyn_02, swing1LimitDegrees=20, swing2LimitDegrees=20, twistLimitDegrees=1}
 {
  bone1Name=band04_dyn_base
  bone2Name=pack
  swing1LimitDegrees=45
  swing1Motion=Free
  swing2LimitDegrees=45
  swing2Motion=Free
  twistLimitDegrees=45
  twistMotion=Free
 }
 {bone1Name=band04_dyn_00, bone2Name=band04_dyn_base, swing1LimitDegrees=15, swing2LimitDegrees=2, twistLimitDegrees=1}
 {bone1Name=band04_dyn_01, bone2Name=band04_dyn_00, swing1LimitDegrees=10, swing2LimitDegrees=10, twistLimitDegrees=1}
 {
  bone1Name=band03_dyn_base
  bone2Name=pack
  swing1LimitDegrees=45
  swing1Motion=Free
  swing2LimitDegrees=45
  swing2Motion=Free
  twistLimitDegrees=45
  twistMotion=Free
 }
 {bone1Name=band03_dyn_00, bone2Name=band03_dyn_base, swing1LimitDegrees=5, swing2LimitDegrees=5, twistLimitDegrees=1}
 {bone1Name=band03_dyn_01, bone2Name=band03_dyn_00, swing1LimitDegrees=10, swing2LimitDegrees=10, twistLimitDegrees=1}
 {bone1Name=band02_dyn_base, bone2Name=pack, swing1LimitDegrees=15, swing2LimitDegrees=15, twistLimitDegrees=1}
 {bone1Name=band02_dyn_01, bone2Name=band02_dyn_base, swing1LimitDegrees=15, swing2LimitDegrees=10, twistLimitDegrees=1}
 {bone1Name=band02_dyn_02, bone2Name=band02_dyn_01, swing1LimitDegrees=1, swing2LimitDegrees=10, twistLimitDegrees=1}
 {bone1Name=band01_dyn_base, bone2Name=pack, swing1LimitDegrees=15, swing2LimitDegrees=15, twistLimitDegrees=1}
 {bone1Name=band01_dyn_01, bone2Name=band01_dyn_base, swing1LimitDegrees=10, swing2LimitDegrees=20, twistLimitDegrees=1}
 {bone1Name=band01_dyn_02, bone2Name=band01_dyn_01, swing1LimitDegrees=1, swing2LimitDegrees=10, twistLimitDegrees=1}
]
